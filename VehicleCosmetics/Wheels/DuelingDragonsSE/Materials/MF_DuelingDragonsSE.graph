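# Material function for the Dueling Dragons SE wheel cosmetic, setting material attributes for a reflective rim with animated colored glow.
# A box-projected cubemap lookup gives desaturated reflections for base color and emissive.
# Time-driven sines on the second UV channel blend glow layers and pulse bands, combining Fresnel, vertex color, color parameters and tread lighting.
# RGBA texture masks place layers and drive metallic and roughness, while a PaintedShading static switch selects re-masked emissive.
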MaterialFunction /VehicleCosmetics/Wheels/DuelingDragonsSE/Materials/MF_DuelingDragonsSE
graph {nodes [
  n1 = NamedRerouteUsage()
  n2 = Constant3Vector() {constant=<rgba 5000 7500 5000 0>}
  n3 = WorldPosition()
  n4 = Transform(None=ReflectionVectorWS())
  n5 = Subtract(B=@n2) {constA=0}
  n6 = Max(A=Divide(A=Subtract(A=@n2, B=@n3.XYZ), B=@n4), B=Divide(A=Subtract(A=@n5, B=@n3.XYZ), B=@n4))
  n7 = Min(A=Min(A=ComponentMask(None=@n6) {r=true}, B=ComponentMask(None=@n6) {g=true}), B=ComponentMask(None=@n6) {b=true})
  n8 = TextureSample(
   UVs=Add(A=WorldPosition().XYZ, B=Multiply(A=@n7, B=@n4))
   Tex=TextureObject() {samplerType=SAMPLERTYPE_LinearColor, texture=&/TC_DuelingDragonsSE_GrassCube}
  ) {texture=/VehicleCosmetics/Wheels/DuelingDragonsSE/Textures/T_Rim_DuelingDragonsSE_RGB}
  n9 = Constant()
  n10 = Saturate(None=Power(Base=Desaturation(None=@n8.RGB, Fraction=@n9) {luminanceFactors=<rgba 0.3 0.59 0.11 0>}))
  n11 = NamedRerouteUsage()
  n12 = Power(Base=NamedRerouteUsage()) {constExponent=1}
  n13 = Add(A=Multiply(A=@n11, B=@n12), B=@n11)
  n14 = NamedRerouteUsage()
  n15 = Multiply(B=Fresnel(Normal=Transform(None=@n14)) {baseReflectFraction=0, exponent=8}) {constA=6}
  n16 = VertexColor()
  n17 = Fresnel(Normal=Transform(None=NamedRerouteUsage())) {baseReflectFraction=0, exponent=1}
  n18 = CameraVectorWS()
  n19 = Power(Base=ComponentMask(None=OneMinus(None=@n18)) {b=true}) {constExponent=0.3}
  n20 = Add(A=Clamp(None=Multiply(B=@n19) {constA=2}) {maxDefault=0.5}, B=NamedRerouteUsage())
  n21 = Multiply(A=Multiply(A=@n20, B=NamedRerouteUsage()), B=ScalarParameter() {defaultValue=6, group=3_InputProperties, parameterName=Tread Lights})
  n22 = Clamp(None=Multiply(A=NamedRerouteUsage(), B=ScalarParameter() {defaultValue=1, group=3_InputProperties, parameterName=TireEmissiveGlow})) {maxDefault=3}
  n23 = Multiply(A=Multiply(A=NamedRerouteUsage(), B=ScalarParameter() {defaultValue=3, group=3_InputProperties, parameterName=ColorPower}), B=@n22)
  n24 = Add(A=Multiply(A=Multiply(A=@n15, B=@n16.G), B=NamedRerouteUsage()), B=Multiply(A=@n17, B=Add(A=@n21, B=@n23)))
  n25 = Add(A=Multiply(A=@n12, B=@n24), B=@n24)
  n26 = Time()
  n27 = ScalarParameter() {defaultValue=0.2, group=3_InputProperties, parameterName=DuelSpeed}
  n28 = Add(A=Multiply(A=@n26, B=Multiply(B=@n27) {constA=-1}), B=ComponentMask(None=TextureCoordinate() {coordinateIndex=1}) {r=true})
  n29 = Multiply(A=Add(A=Sine(None=@n28))) {constB=0.5}
  n30 = Clamp(None=Power(Base=LinearInterpolate(A=@n13, B=@n25, Alpha=@n29)))
  n31 = TextureCoordinate() {coordinateIndex=1}
  n32 = Multiply(A=ScalarParameter() {defaultValue=0.5, group=3_InputProperties, parameterName=BandSpeed}) {constA=-1, constB=-1}
  n33 = Sine(None=Add(A=ComponentMask(None=@n31) {g=true}, B=Multiply(A=@n32, B=@n26)))
  n34 = Add(A=Multiply(A=Add(A=@n33)) {constB=0.5})
  n35 = Multiply(A=ScalarParameter() {defaultValue=1, group=3_InputProperties, parameterName=InnerEmissive}, B=NamedRerouteUsage())
  n36 = NamedRerouteUsage()
  n37 = NamedRerouteUsage()
  n38 = NamedRerouteUsage()
  n39 = NamedRerouteUsage()
  n40 = Saturate(None=Add(A=NamedRerouteUsage(), B=NamedRerouteUsage()))
  n41 = LinearInterpolate(
   A=Add(A=@n10, B=Multiply(A=@n34, B=@n30))
   B=LinearInterpolate(
    A=NamedRerouteDeclaration(None=@n35) {name=BW_Lines_Out, nodeColor=<rgba 0.6406 0 1 1>}
    B=LinearInterpolate(A=@n36, B=@n24, Alpha=@n37) {constB=10}
    Alpha=Add(A=@n38, B=@n39)
   )
   Alpha=@n40
  )
  n42 = NamedRerouteUsage()
  n43 = Clamp(None=OneMinus(None=@n42)) {maxDefault=0.8}
  n44 = Add(A=NamedRerouteUsage()) {constB=-0.01}
  n45 = ConstantBiasScale(None=Saturate(None=Multiply(A=@n43, B=@n44))) {bias=-0.5, scale=2}
  n46 = Saturate(None=@n17)
  n47 = ConstantBiasScale(None=Multiply(A=@n42, B=Multiply(A=@n46) {constB=10})) {bias=-0.1, scale=2}
  n48 = Saturate(None=Multiply(A=NamedRerouteUsage()) {constB=20})
  n49 = LinearInterpolate(
   A=Multiply(A=Multiply(A=@n1, B=@n41)) {constB=0.1}
   B=@n1
   Alpha=Saturate(None=LinearInterpolate(A=@n45, B=@n47, Alpha=@n48) {desc=Combined Inverted Mask})
  )
  n50 = Saturate(None=NamedRerouteUsage())
  n51 = OneMinus(None=@n50)
  n52 = SetMaterialAttributes!(
   MaterialAttributes
   BaseColor=LinearInterpolate(A=@n10, Alpha=@n40) {constB=0}
   EmissiveColor=StaticSwitchParameter(True=ConstantBiasScale(None=@n49) {bias=0, scale=2}, False=@n41) {group=3_InputProperties, parameterName=PaintedShading}
   Normal=NamedRerouteUsage()
   Metallic=Saturate(None=Add(A=@n51) {constB=-0.8})
   Specular
   Roughness=Saturate(None=Add(A=@n50) {constB=0.2})
  ) {attributeSetTypes [
    #1
    #2
    #3
    #4
    #5
    #6
   ]}
  FunctionOutput(None=@n52)
  n53 = TextureSampleParameter2D() {
   group=1_InputTextures
   parameterName=RimNormal
   samplerType=SAMPLERTYPE_Normal
   texture=/VehicleCosmetics/Wheels/DuelingDragonsSE/Textures/T_Rim_DuelingDragonsSE_N
  }
  NamedRerouteDeclaration(None=@n53.RGB) {name=RimNormal, nodeColor=<rgba 1 0 0.7266 1>}
  NamedRerouteDeclaration(None=VectorParameter() {group=2_InputColor, parameterName=CustomColor, defaultValue=<rgba 0 0.3797 1 3>}.RGB) {name=CustomColor, nodeColor=<rgba 0 1 0.9531 1>}
  NamedRerouteDeclaration(None=VectorParameter() {group=2_InputColor, parameterName=AccentColor, defaultValue=<rgba 1 0.0436 0.6445 3>}.RGB) {name=AccentColor, nodeColor=<rgba 1 0.6328 0 1>}
  n54 = TextureSampleParameter2D() {group=1_InputTextures, parameterName=MaskRGB, texture=&T_Rim_DuelingDragonsSE_RGB}
  NamedRerouteDeclaration(None=@n54.R) {name=R_Mask, nodeColor=<rgba 0.9688 1 0 1>}
  NamedRerouteDeclaration(None=@n54.G) {name=G_Mask, nodeColor=<rgba 0.7109 0 1 1>}
  NamedRerouteDeclaration(None=@n54.B) {name=B_Mask, nodeColor=<rgba 0 1 0.3906 1>}
  NamedRerouteDeclaration(None=@n54.A) {name=A_Mask, nodeColor=<rgba 1 0.0703 0 1>}
  n55 = Transform(None=NamedRerouteUsage())
  n56 = Multiply(A=Multiply(B=Fresnel(Normal=@n55) {baseReflectFraction=1, exponent=8}) {constA=6}, B=VertexColor().G) {constA=6}
  n57 = NamedRerouteUsage()
  n58 = Fresnel(Normal=Transform(None=NamedRerouteUsage())) {baseReflectFraction=0, exponent=1}
  n59 = NamedRerouteUsage()
  n60 = ScalarParameter() {defaultValue=3, group=3_InputProperties, parameterName=ColorPower}
  n61 = Add(
   A=Multiply(A=Multiply(A=@n20, B=@n59), B=ScalarParameter() {defaultValue=6, group=3_InputProperties, parameterName=Tread Lights})
   B=Multiply(A=Multiply(A=@n59, B=@n60), B=@n22)
  )
  NamedRerouteDeclaration(None=Add(A=Multiply(A=@n56, B=@n57) {constA=6}, B=Multiply(A=@n58, B=@n61))) {name=CustomColorGlow_Out, nodeColor=<rgba 0 1 0.3203 1>}
 ]}
